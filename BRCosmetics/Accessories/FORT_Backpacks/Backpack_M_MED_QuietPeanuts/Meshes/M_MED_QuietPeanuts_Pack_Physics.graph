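# Physics asset for the QuietPeanuts backpack cosmetic, defining collision and joint limits for its two swords and a ribbon with two straps.
# Each sword is a long capsule held to a small master base body by tight swing and twist limits.
# The ribbon base hangs from the second sword and branches into two four-segment chains of small capsules.
# Strap joints allow much wider swing than the swords while keeping twist tight.
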PhysicsAsset /BRCosmetics/Accessories/FORT_Backpacks/Backpack_M_MED_QuietPeanuts/Meshes/M_MED_QuietPeanuts_Pack_Physics
bodies {
 SwordBase_MSTR [{shapeType=Capsule, length=0.505, radius=0.505, rotation=<rot 0 -90 0>}]
 Sword_01_Base_jnt [{shapeType=Capsule, length=30, radius=2, center=<-5 0 -0.5>, rotation=<rot 0 90 95>}]
 Sword_02_Base_jnt [{shapeType=Capsule, length=30, radius=2, center=<-3 0 0.5>, rotation=<rot 0 90 85>}]
 dyn_01RibbonStrap_01_jnt [{shapeType=Capsule, length=1, radius=1, shapeName=dyn_01RibbonStrap_01_jnt_capsule, center=<2 0 0>, rotation=<rot 90 0 0>}]
 dyn_01RibbonStrap_02_jnt [{shapeType=Capsule, length=1, radius=1, shapeName=dyn_01RibbonStrap_02_jnt_capsule, center=<2 0 0>, rotation=<rot 90 0 0>}]
 dyn_01RibbonStrap_03_jnt [{shapeType=Capsule, length=1, radius=1, shapeName=dyn_01RibbonStrap_03_jnt_capsule, center=<2 0 0>, rotation=<rot 90 0 0>}]
 dyn_01RibbonStrap_04_jnt [{shapeType=Capsule, length=1, radius=1, shapeName=dyn_01RibbonStrap_04_jnt_capsule, center=<2 0 0>, rotation=<rot 90 0 0>}]
 dyn_02RibbonStrap_01_jnt [{shapeType=Capsule, length=1, radius=1, shapeName=dyn_02RibbonStrap_01_jnt_capsule, center=<2 0 0>, rotation=<rot 90 0 0>}]
 dyn_02RibbonStrap_02_jnt [{shapeType=Capsule, length=1, radius=1, shapeName=dyn_02RibbonStrap_02_jnt_capsule, center=<2 0 0>, rotation=<rot 90 0 0>}]
 dyn_02RibbonStrap_03_jnt [{shapeType=Capsule, length=1, radius=1, shapeName=dyn_02RibbonStrap_03_jnt_capsule, center=<2 0 0>, rotation=<rot 90 0 0>}]
 dyn_02RibbonStrap_04_jnt [{shapeType=Capsule, length=1, radius=1, shapeName=dyn_02RibbonStrap_04_jnt_capsule, center=<2 0 0>, rotation=<rot 90 0 0>}]
 dyn_Ribbon_Base_jnt [{shapeType=Capsule, length=1, radius=1, shapeName=dyn_Ribbon_Base_jnt_capsule, center=<2 0 0>, rotation=<rot 90 0 0>}]
}
constraints [
 {bone1Name=Sword_01_Base_jnt, bone2Name=SwordBase_MSTR, swing1LimitDegrees=2, swing2LimitDegrees=6, twistLimitDegrees=2}
 {bone1Name=Sword_02_Base_jnt, bone2Name=SwordBase_MSTR, swing1LimitDegrees=2, swing2LimitDegrees=6, twistLimitDegrees=2}
 {bone1Name=dyn_Ribbon_Base_jnt, bone2Name=Sword_02_Base_jnt, swing1LimitDegrees=2, swing2LimitDegrees=55, twistLimitDegrees=2}
 {bone1Name=dyn_01RibbonStrap_01_jnt, bone2Name=dyn_Ribbon_Base_jnt, swing1LimitDegrees=15, swing2LimitDegrees=55, twistLimitDegrees=2}
 {bone1Name=dyn_01RibbonStrap_02_jnt, bone2Name=dyn_01RibbonStrap_01_jnt, swing1LimitDegrees=45, swing2LimitDegrees=45, twistLimitDegrees=2}
 {bone1Name=dyn_01RibbonStrap_03_jnt, bone2Name=dyn_01RibbonStrap_02_jnt, swing1LimitDegrees=45, swing2LimitDegrees=45, twistLimitDegrees=2}
 {bone1Name=dyn_01RibbonStrap_04_jnt, bone2Name=dyn_01RibbonStrap_03_jnt, swing1LimitDegrees=45, swing2LimitDegrees=45, twistLimitDegrees=2}
 {bone1Name=dyn_02RibbonStrap_01_jnt, bone2Name=dyn_Ribbon_Base_jnt, swing1LimitDegrees=15, swing2LimitDegrees=55, twistLimitDegrees=2}
 {bone1Name=dyn_02RibbonStrap_02_jnt, bone2Name=dyn_02RibbonStrap_01_jnt, swing1LimitDegrees=45, swing2LimitDegrees=45, twistLimitDegrees=2}
 {bone1Name=dyn_02RibbonStrap_03_jnt, bone2Name=dyn_02RibbonStrap_02_jnt, swing1LimitDegrees=45, swing2LimitDegrees=45, twistLimitDegrees=2}
 {bone1Name=dyn_02RibbonStrap_04_jnt, bone2Name=dyn_02RibbonStrap_03_jnt, swing1LimitDegrees=45, swing2LimitDegrees=45, twistLimitDegrees=2}
]
